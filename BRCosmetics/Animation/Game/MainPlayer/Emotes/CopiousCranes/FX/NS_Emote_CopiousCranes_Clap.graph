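# Niagara effect for the clap in the CopiousCranes emote: a brief flash with a spray of glinting sparkles, both placed on the character's skeletal mesh.
# MainShape bursts a very short-lived shape, layering four sprite renderers and a mesh, with derived corona, dark glow and flare sizes.
# Sparkles_Glint bursts randomly sized sprites outward from the sampled point, slowed by drag and flickering through alpha curves.
# Time-of-day functions tint both emitters, bluer at night.
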NiagaraSystem /BRCosmetics/Animation/Game/MainPlayer/Emotes/CopiousCranes/FX/NS_Emote_CopiousCranes_Clap
dependencies {
 usedDataInterfaces [
  &NiagaraDataInterfaceCurve
  &NiagaraDataInterfaceArrayFloat3
  &NiagaraDataInterfaceColorCurve
  &NiagaraDataInterfaceArrayFloat
  &NiagaraDataInterfaceSkeletalMesh
 ]
 usedDynamicInputs []
 usedModules [
  &self:SystemSpawnScript.SystemScriptSource.SystemScriptGraph.NiagaraNodeAssignment_0.SetVariables_#1
  &SystemState
 ]
}
emitters {
 MainShape {
  emitterUpdateScript [
   &EmitterState(Inactive Response=::NewEnumerator1, Life Cycle Mode=::NewEnumerator1, Loop Behavior=::NewEnumerator1, Loop Duration=1, UseLoopDelay=false)
   &SpawnBurst_Instantaneous(Use Spawn Probability=false, UseLoopCountLimit=false)
  ]
  particleSpawnScript [
   /SetVariables(Particles.VisibilityTag=&SpawnGroup()) {isSetParameters=true}
   &InitializeParticle(
    Color=&Multiply_LinearColorByFloat(
     Linear Color=<rgba 1 0 0.1038 1>
     Float=&FNTimeOfDayScalar(Night Scale=1.4, Morning Scale=1.4, Day Scale=1.4, Evening Scale=1.4)
    )
    Lifetime=0.13
    Mass Mode=::NewEnumerator1
    Mesh Scale Mode=::NewEnumerator3
    Mesh Uniform Scale=&FloatFromCurve(
     FloatCurve=DataInterface() {propertyValues {Curve {keys [{tangentMode=RCTM_User, time=0, value=1}, {tangentMode=RCTM_User, time=1, value=2}]}}}
     CurveIndex=&ReturnNormalizedExecIndex()
     Scale Curve=5
    )
    Position=&SimulationPosition()
    Position Mode=::NewEnumerator1
    Position Offset=<0 0 125>
    Sprite Size Mode=::NewEnumerator3
    Uniform Sprite Size=130
    UsePositionOffset=false
    Write Material Random=true
   ) {module=InitializeParticle001}
   &SkeletalMeshLocation(Bone Sampling Mode=::NewEnumerator4, Inherit Velocity=false, MaskSpawnGroup=false, Skeletal Mesh=$System.Skeletal Mesh)
   &CameraOffset(Camera Offset Amount=5)
   &InitialMeshOrientation(
    Mesh Orientation Mode=::NewEnumerator2
    Rotation=&SelectVectorFromArray(
     Direct Array Index=&SpawnGroup()
     Vector Selection Array=DataInterface() {propertyValues {FloatData [<-0.25 0 0>, <0.5 0 0>]}}
    )
    Rotation Coordinate Space=::NewEnumerator2
    Use Rotation Vector=true
   )
  ]
  particleUpdateScript [
   &ParticleState()
   &SolveForcesAndVelocity(Clamp Velocity=false, Limit Acceleration=false)
   &ScaleColor(
    Scale Alpha=&FloatFromCurve(FloatCurve=DataInterface() {propertyValues {Curve {keys [
         {arriveTangent=6.4019, interpMode=RCIM_Cubic, leaveTangent=6.4019, tangentMode=RCTM_User, time=0, value=0}
         {arriveTangent=0.002, interpMode=RCIM_Cubic, leaveTangent=0.002, time=0.2, value=0.7}
         {arriveTangent=-0.9006, interpMode=RCIM_Cubic, leaveTangent=-0.9006, tangentMode=RCTM_User, time=1.0011, value=0.002}
        ]}}})
    Scale RGB=&MakeVectorFromLinearColorRGB(Color=&FNTimeOfDayColor(Night Color=<rgba 0.15 0.7167 1 1>, Morning Color=<rgba 1 1 1 1>, Day Color=<rgba 1 1 1 1>, Evening Color=<rgba 1 1 1 1>))
    ScaleA=true
    ScaleRGB=true
   )
   &ScaleSpriteSize(Scale Factor=&Multiply_Vector2D_ByFloat(
     Float=&FloatFromCurve(
      FloatCurve=DataInterface() {propertyValues {Curve {keys [
          {arriveTangent=0.7585, interpMode=RCIM_Cubic, leaveTangent=0.7585, tangentMode=RCTM_User, time=0, value=0.4}
          {interpMode=RCIM_Cubic, tangentMode=RCTM_User, time=1, value=1}
         ]}}}
      Scale Curve=1.25
     )
     Vector2D=<1 1>
    ))
   &ScaleMeshSize(Scale Factor=&VectorFromFloat(Value=&FloatFromCurve(
      FloatCurve=DataInterface() {propertyValues {Curve {keys [
          {arriveTangent=4.1785, interpMode=RCIM_Cubic, leaveTangent=4.1785, tangentMode=RCTM_User, time=0, value=0}
          {arriveTangent=1.1179, interpMode=RCIM_Cubic, leaveTangent=1.1179, tangentMode=RCTM_User, time=0.2042, value=0.6228}
          {arriveTangent=0.0224, interpMode=RCIM_Cubic, leaveTangent=0.0224, tangentMode=RCTM_User, time=1, value=1}
         ]}}}
      Scale Curve=0.35
     )))
   /SetVariables(
    Particles.DynamicParameter_01:Vector4f=&MakeVector4(
     X=&RandomRangeFloat(Minimum=0, Maximum=1)
     Y=2.5
     Z=&RandomRangeFloat(Minimum=2, Maximum=3)
     W=0
    )
    Particles.DynamicParameter_02:Vector4f=&MakeVector4(X=0, Y=&RandomRangeFloat(Minimum=0, Maximum=1), Z=0, W=0)
   ) {isSetParameters=true}
   /SetVariables(Particles.Corona_Size:Vector2f=&Multiply_Vector2D_ByFloat(Float=0.5, Vector2D=$Particles.SpriteSize)) {isSetParameters=true}
   /SetVariables(Particles.DarkGlowSize:Vector2f=&Multiply_Vector2D_ByFloat(Float=0.75, Vector2D=$Particles.SpriteSize)) {isSetParameters=true}
   /SetVariables(Particles.DargGlowColor:LinearColor=&ColorFromCurve(ColorCurve=DataInterface() {propertyValues {
       AlphaCurve {keys [<at 0.4134 0.7 RCIM_Cubic>,<at 1 0 RCIM_Cubic>]}
       BlueCurve {keys [<at 0 0.02 RCIM_Cubic>]}
       GreenCurve {keys [<at 0 0.0127 RCIM_Cubic>]}
       RedCurve {keys [<at 0 0.0066 RCIM_Cubic>]}
      }})) {isSetParameters=true}
   /SetVariables(Particles.CentralFlareSize:Vector2f=&Multiply_Vector2D_ByFloat(Float=2, Vector2D=$Particles.SpriteSize)) {isSetParameters=true}
   /SetVariables(Particles.CentralFlareColor:LinearColor=&MultiplyLinearColor(
     Linear Color=&ColorFromCurve(
      ColorCurve=DataInterface() {propertyValues {
        AlphaCurve {keys [<at 0 1>,<at 1 0>]}
        BlueCurve {keys [<at 0.2962 0.1661>,<at 0.5678 0.2>,<at 0.8157 0.1647>]}
        GreenCurve {keys [<at 0.2962 0.1069>,<at 0.5678 0.2>,<at 0.8157 0.1059>]}
        RedCurve {keys [<at 0.2962 0>,<at 0.5678 1>,<at 0.8157 0>]}
       }}
      CurveIndex=&Multiply_Float(A=$Particles.NormalizedAge, B=2)
     )
     Scale Factor=&MakeFloat4FromVectorAndFloat("Vector (RGB)"=&VectorFromFloat(Value=2), "Float (Alpha)"=0.5)
    )) {isSetParameters=true}
  ]
  rendererClasses [&NiagaraSpriteRendererProperties, &NiagaraMeshRendererProperties]
  renderers [
   {rendererClass=&NiagaraSpriteRendererProperties}
   {rendererClass=&NiagaraSpriteRendererProperties, rendererIndex=1}
   {rendererClass=&NiagaraSpriteRendererProperties, rendererIndex=2}
   {rendererClass=&NiagaraSpriteRendererProperties, rendererIndex=3}
   {rendererClass=&NiagaraMeshRendererProperties, rendererIndex=4}
  ]
 }
 Sparkles_Glint {
  emitterUpdateScript [
   &EmitterState(Life Cycle Mode=::NewEnumerator1, Loop Behavior=::NewEnumerator1, Loop Duration=1, UseLoopDelay=false)
   &SpawnBurst_Instantaneous(
    Spawn Count=&RandomRangeInt(Minimum=25, Maximum=35)
    Use Spawn Probability=false
    UseLoopCountLimit=false
   )
  ]
  particleSpawnScript [
   /SetVariables(Particles.VisibilityTag=&RandomRangeInt(Minimum=0, Maximum=2)) {isSetParameters=true}
   &InitializeParticle(
    AdjustAlpha=false
    AdjustHue=false
    AdjustSaturation=false
    AdjustValue=false
    Color=<rgba 0.2467 0.6813 1 1>
    Color Mode=::NewEnumerator3
    Hue Shift Range:Vector2f=<-1 1>
    Lifetime Max=0.2
    Lifetime Min=0.4
    Lifetime Mode=::NewEnumerator1
    Sprite Size Mode=::NewEnumerator3
    Uniform Sprite Size=&Multiply_Float(
     A=&Multiply_Float(
      A=&SelectFloatFromArray(Direct Array Index=$Particles.VisibilityTag, Float Selection Array=DataInterface() {propertyValues {FloatData [60,30,15]}})
      B=&RandomRangeFloat(Minimum=1.5, Maximum=2)
     )
     B=0.9
    )
    UsePositionOffset=true
    Write Material Random=true
   )
   &CameraOffset(Camera Offset Amount=15)
   &SkeletalMeshLocation(Bone Sampling Mode=::NewEnumerator4, Inherit Velocity=false, MaskSpawnGroup=false, Skeletal Mesh=$System.Skeletal Mesh)
   &ShapeLocation(MaskSpawnGroup=false, Override Seed=false, Sphere Radius=30, UseOwnerScale=true)
   &AddVelocity(
    Module.Use Falloff Distance=false
    Override Seed=false
    Velocity Mode=::NewEnumerator1
    Velocity Origin=$Output.SkeletalMeshLocation.SampledPosition
    Velocity Speed=&RandomRangeFloat(Minimum=50, Maximum=300)
   )
  ]
  particleUpdateScript [
   &ParticleState()
   &ScaleColor(
    Scale Alpha=&Multiply_Float(
     A=&FloatFromCurve(
      FloatCurve=DataInterface() {propertyValues {Curve {keys [
          {interpMode=RCIM_Cubic, tangentMode=RCTM_User, time=0, value=0}
          {arriveTangent=-2.4058, interpMode=RCIM_Cubic, leaveTangent=-2.4058, tangentMode=RCTM_User, time=0.26, value=0.8421}
          {arriveTangent=-1.1775, interpMode=RCIM_Cubic, leaveTangent=-1.1775, time=0.4613, value=0.0645}
          {arriveTangent=-0.1197, interpMode=RCIM_Cubic, leaveTangent=-0.1197, time=0.5898, value=0.4537}
          <at 1 0 RCIM_Cubic>
         ]}}}
      Scale Curve=&RandomRangeFloat(Minimum=0.2, Maximum=0.4)
     )
     B=&FloatFromCurve(
      FloatCurve=DataInterface() {propertyValues {Curve {keys [{tangentMode=RCTM_User, time=0, value=0}, {tangentMode=RCTM_User, time=1, value=1}]}}}
      CurveIndex=&Multiply_Float(
       A=&MakeFloatFromVector(Vector=Linked() {linkedVariable {name=Particles.Position, type=NiagaraPosition}}, Channel=::NewEnumerator2)
       B=$Engine.System.Age
      )
     )
    )
    Scale RGB=&MakeVectorFromLinearColorRGB(Color=&FNTimeOfDayColor(Night Color=<rgba 0.3 0.6267 1 1>, Morning Color=<rgba 1 1 1 1>, Day Color=<rgba 1 1 1 1>, Evening Color=<rgba 1 1 1 1>))
    ScaleA=true
    ScaleRGB=true
   )
   &ScaleSpriteSize(Scale Factor=&Multiply_Vector2D_ByFloat(
     Float=1
     Vector2D=&MakeVector2D(
      X=&FloatFromCurve(FloatCurve=DataInterface() {propertyValues {Curve {keys [{interpMode=RCIM_Cubic, tangentMode=RCTM_User, time=0, value=0}, {interpMode=RCIM_Cubic, tangentMode=RCTM_User, time=1, value=1}]}}})
      Y=&FloatFromCurve(FloatCurve=DataInterface() {propertyValues {Curve {keys [{interpMode=RCIM_Cubic, tangentMode=RCTM_User, time=0, value=1}, {interpMode=RCIM_Cubic, tangentMode=RCTM_User, time=1, value=0.6493}]}}})
     )
    ))
   /SetVariables(
    Particles.Color=&MultiplyLinearColor(
     Linear Color=$Particles.Color
     Scale Factor=&MakeVector4_FromLinearColor(LinearColor=<rgba 8.0099 9.8007 10 1>)
    )
    Particles.Color02:LinearColor=&MultiplyLinearColor(
     Linear Color=$Particles.Color
     Scale Factor=&MakeVector4_FromLinearColor(LinearColor=<rgba 0.9441 0.8079 1 1>)
    )
    Particles.Color03:LinearColor=&MultiplyLinearColor(
     Linear Color=$Particles.Color
     Scale Factor=&MakeVector4_FromLinearColor(LinearColor=<rgba 0.15 0.15 0.15 1>)
    )
    Particles.SpriteSize02:Vector2f=&Multiply_Vector2(A=$Particles.SpriteSize, B=&Vector2DFromFloat(Value=0.3))
   ) {isSetParameters=true}
   &Drag(Drag=&RandomRangeFloat(Minimum=4, Maximum=5), Use Linear Drag=true, Use Rotational Drag=true)
   &SolveForcesAndVelocity(Clamp Velocity=false, Limit Acceleration=false)
  ]
  rendererClasses [&NiagaraSpriteRendererProperties]
  renderers [{rendererClass=&NiagaraSpriteRendererProperties}]
 }
}
summary {
 systemName=NS_Emote_CopiousCranes_Clap
 emitters [
  {emitterName=MainShape, rendererClasses [&NiagaraSpriteRendererProperties, &NiagaraMeshRendererProperties]}
  {emitterName=Sparkles_Glint, rendererClasses [&NiagaraSpriteRendererProperties]}
 ]
 userVariables [
  User.bIsFrontEnd {type=NiagaraBool}
  User.bIsFrontEndPreview {type=NiagaraBool}
 ]
}
