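# Niagara wind effect for the CopiousCranes emote, combining mesh, sprite and ribbon emitters.
# A rotating wind disc and short-lived streak meshes spawn near a fixed system offset, using time-of-day colors and front-end modifiers.
# Sparkle and smoke sprites share staggered bursts, each spawn group pinned to its own skeletal mesh bone or socket.
# Five ribbon trails follow a separate skeletal mesh input, with spawn-group-masked curves over emitter age staggering their fade-out.
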NiagaraSystem /BRCosmetics/Animation/Game/MainPlayer/Emotes/CopiousCranes/FX/NS_Emote_CopiousCranes_Wind
dependencies {
 usedDataInterfaces [&NiagaraDataInterfaceCurve, &NiagaraDataInterfaceArrayColor]
 usedDynamicInputs []
 usedModules [
  &self:SystemSpawnScript.SystemScriptSource.SystemScriptGraph.NiagaraNodeAssignment_0.SetVariables_#1
  &self:SystemSpawnScript.SystemScriptSource.SystemScriptGraph.NiagaraNodeAssignment_2.SetVariables_#2
  &self:SystemSpawnScript.SystemScriptSource.SystemScriptGraph.NiagaraNodeAssignment_3.SetVariables_#3
  &SystemState
 ]
}
emitters {
 NE_Sparkles {
  emitterUpdateScript [
   &EmitterState(Life Cycle Mode=::NewEnumerator1, Loop Behavior=::NewEnumerator1, Loop Delay=0.2, Loop Duration=2, UseLoopDelay=true)
   &SpawnBurst_Instantaneous(Spawn Group=5, Spawn Time=1.212, Use Spawn Probability=false, UseLoopCountLimit=false) {module=SpawnBurst_Instantaneous005}
   &SpawnBurst_Instantaneous(Spawn Group=4, Spawn Time=1.1512, Use Spawn Probability=false, UseLoopCountLimit=false) {module=SpawnBurst_Instantaneous004}
   &SpawnBurst_Instantaneous(Spawn Group=3, Spawn Time=0.8532, Use Spawn Probability=false, UseLoopCountLimit=false) {module=SpawnBurst_Instantaneous003}
   &SpawnBurst_Instantaneous(Spawn Group=2, Spawn Time=1.0228, Use Spawn Probability=false, UseLoopCountLimit=false) {module=SpawnBurst_Instantaneous002}
   &SpawnBurst_Instantaneous(Spawn Group=1, Spawn Time=0.69, Use Spawn Probability=false, UseLoopCountLimit=false) {module=SpawnBurst_Instantaneous001}
   &SpawnBurst_Instantaneous(Spawn Count=1, Spawn Time=0.8556, Use Spawn Probability=false)
  ]
  particleSpawnScript [
   &InitializeParticle(
    Color=<rgba 1.9537 3.8243 5 1>
    Lifetime=0.2
    Sprite Rotation Mode=::NewEnumerator2
    Sprite Size Mode=::NewEnumerator1
    Sprite UV Mode=::NewEnumerator4
    Uniform Sprite Size Max=20
    Uniform Sprite Size Min=15
    UsePositionOffset=false
    Write Material Random=true
   ) {module=InitializeParticle001}
   &SkeletalMeshLocation(
    Bone / Socket Index=&SpawnGroup()
    Bone Sampling Mode=::NewEnumerator4
    MaskSpawnGroup=false
    Skeletal Mesh=$System.Skeletal Mesh
   )
   &ShapeLocation(MaskSpawnGroup=false, Override Seed=false, Shape Origin=$Particles.Position, Sphere Radius=1, UseOwnerScale=true)
   &CameraOffset(Camera Offset Amount=10)
  ]
  particleUpdateScript [
   &ParticleState()
   &SpriteRotationRate(Rotation Rate=&RandomRangeFloat(Minimum=500, Maximum=400))
   &SolveForcesAndVelocity(Clamp Velocity=false, Limit Acceleration=false)
   &ScaleColor(
    Scale Alpha=&FloatFromCurve(FloatCurve=DataInterface() {propertyValues {Curve {keys [<at -0.0092 0.0144>,<at 0.0967 0.9767>,<at 1.0011 0.002>]}}})
    ScaleA=true
    ScaleRGB=true
   )
   &ScaleSpriteSize(Scale Factor=&Vector2DFromFloat(Value=&FloatFromCurve(
      FloatCurve=DataInterface() {propertyValues {Curve {keys [
          {
           arriveTangent=-0.0355
           arriveTangentWeight=0.0945
           interpMode=RCIM_Cubic
           leaveTangent=3.9164
           leaveTangentWeight=0.4355
           tangentMode=RCTM_Break
           tangentWeightMode=RCTWM_WeightedBoth
           time=0
           value=0.7085
          }
          {
           arriveTangent=0.0048
           arriveTangentWeight=0.1184
           interpMode=RCIM_Cubic
           leaveTangent=0.0337
           leaveTangentWeight=0.0875
           tangentMode=RCTM_Break
           time=0.2106
           value=1.4624
          }
          <at 1.0011 0.002 RCIM_Cubic>
         ]}}}
      Scale Curve=&RandomRangeFloat(Minimum=0.8, Maximum=1)
     )))
  ]
  rendererClasses [&NiagaraSpriteRendererProperties]
  renderers [{rendererClass=&NiagaraSpriteRendererProperties}]
 }
 NE_WindDisc {
  emitterUpdateScript [
   &EmitterState(Life Cycle Mode=::NewEnumerator1, Loop Behavior=::NewEnumerator1, Loop Delay=0.3, Loop Duration=0.4, UseLoopDelay=true)
   &SpawnBurst_Instantaneous(Use Spawn Probability=false, UseLoopCountLimit=false)
  ]
  particleSpawnScript [
   &InitializeParticle(
    Color=&FNTimeOfDayColor(
     Night Color=<rgba 0.2043 0.6186 1 2>
     Morning Color=<rgba 0.4589 0.7406 1 2>
     Day Color=<rgba 0.4589 0.7406 1 2>
     Evening Color=<rgba 0.2043 0.6186 1 2>
    )
    Lifetime=0.9
    Mesh Scale Mode=::NewEnumerator3
    Mesh Uniform Scale=9
    UsePositionOffset=false
    Write Material Random=true
   )
   &InitialMeshOrientation(Mesh Orientation Mode=::NewEnumerator2, Orientation Coordinate Space=::Local, Rotation=<-0.318 -0.2963 0.1807>, Use Rotation Vector=true)
   &SystemLocation(MaskSpawnGroup=false, Offset=<0 -37.2361 12.6553>)
  ]
  particleUpdateScript [
   &ParticleState()
   &DynamicMaterialParameters(
    Index 0 Param 1=&FloatFromCurve(FloatCurve=DataInterface() {propertyValues {Curve {keys [{interpMode=RCIM_Cubic, tangentMode=RCTM_User, time=0.5, value=0}, {interpMode=RCIM_Cubic, tangentMode=RCTM_User, time=1, value=1}]}}})
    Index 0 Param 3=1
    Module.Param0WriteEnabled=true
    Module.Param1WriteEnabled=true
    Module.Param2WriteEnabled=true
    Module.Param3WriteEnabled=true
   )
   /SetVariables(
    Particles.XY:NiagaraFloat=&FloatFromCurve(
     FloatCurve=DataInterface() {propertyValues {Curve {keys [
         {arriveTangent=0.6321, interpMode=RCIM_Cubic, leaveTangent=0.6321, tangentMode=RCTM_User, time=0, value=0.2}
         <at 0.4 0.5681>
         {arriveTangent=0.4623, interpMode=RCIM_Cubic, leaveTangent=0.4623, tangentMode=RCTM_User, time=1, value=0.5948}
        ]}}}
     CurveIndex=$Particles.Age
     Scale Curve=0.4
    )
    Particles.Z:NiagaraFloat=&FloatFromCurve(
     FloatCurve=DataInterface() {propertyValues {Curve {keys [
         {arriveTangent=1.8479, interpMode=RCIM_Cubic, leaveTangent=1.8479, tangentMode=RCTM_User, time=0, value=0}
         {arriveTangent=0.2802, interpMode=RCIM_Cubic, leaveTangent=0.2802, tangentMode=RCTM_User, time=0.4, value=0.2907}
         {arriveTangentWeight=0.2223, interpMode=RCIM_Cubic, tangentMode=RCTM_User, time=1, value=0.3405}
        ]}}}
     CurveIndex=$Particles.Age
    )
   ) {isSetParameters=true}
   &ScaleMeshSize(Scale Factor=&MakeVector(X=$Particles.XY, Y=$Particles.XY, Z=$Particles.Z))
   &UpdateMeshOrientation(Rotation Rate=0.9, Rotation Vector=<0 0 1>)
   &ScaleMeshSize(Scale Factor=&Multiply_VectorByFloat(
     Vector=&MakeVector(
      X=1
      Y=1
      Z=&FloatFromCurve(
       FloatCurve=DataInterface() {propertyValues {Curve {keys [{interpMode=RCIM_Cubic, leaveTangent=0.4757, tangentMode=RCTM_Break, time=0, value=1}, <at 1 1.3553>]}}}
       CurveIndex=$Particles.Age
      )
     )
     Float=0.8
    )) {module=ScaleMeshSize001}
   &ScaleColor(
    Scale Alpha=&FloatFromCurve(
     FloatCurve=DataInterface() {propertyValues {Curve {keys [
         {arriveTangent=8.2356, interpMode=RCIM_Cubic, leaveTangent=8.2356, tangentMode=RCTM_User, time=0, value=0}
         {arriveTangent=0.1969, interpMode=RCIM_Cubic, leaveTangent=0.1969, tangentMode=RCTM_User, time=0.2223, value=1}
         {arriveTangent=-1.2858, interpMode=RCIM_Cubic, leaveTangent=-1.2858, time=0.5, value=0.6}
         {arriveTangent=-0.7636, interpMode=RCIM_Cubic, leaveTangent=-0.7636, tangentMode=RCTM_User, time=1, value=0}
        ]}}}
     Scale Curve=&Multiply_Float(A=0.5, B=&FN_FrontEndFloatModifier(Gameplay=1, Front End=1))
    )
    Scale RGB=&Multiply_VectorByFloat(Vector=<1 1 1>, Float=0.7)
    ScaleA=true
    ScaleRGB=true
   )
  ]
  rendererClasses [&NiagaraMeshRendererProperties]
  renderers [{rendererClass=&NiagaraMeshRendererProperties, rendererIndex=0}]
 }
 NE_WindStreaks {
  emitterUpdateScript [
   &EmitterState(Life Cycle Mode=::NewEnumerator1, Loop Behavior=::NewEnumerator1, Loop Delay=0.1, Loop Duration=0.7, UseLoopDelay=true)
   &SpawnRate(
    SpawnRate=&FloatFromCurve(
     FloatCurve=DataInterface() {propertyValues {Curve {keys [
         {
          arriveTangent=-0.0945
          arriveTangentWeight=0.7825
          interpMode=RCIM_Cubic
          leaveTangent=-0.0945
          leaveTangentWeight=0.7825
          tangentMode=RCTM_User
          tangentWeightMode=RCTWM_WeightedBoth
          time=0
          value=3.5843
         }
         {arriveTangent=-12.4864, interpMode=RCIM_Cubic, leaveTangent=-12.4864, tangentMode=RCTM_User, time=1, value=0}
        ]}}}
     CurveIndex=$Emitter.NormalizedLoopAge
    )
    Use Spawn Probability=false
   ) {module=SpawnRate001}
  ]
  particleSpawnScript [
   &InitializeParticle(
    Color=&MakeCustomLinearColorFromBool(
     Bool=&RandomBool()
     True Color=<rgba 1 0.402 0.3743 1>
     False Color=&FNTimeOfDayColor(
      Night Color=<rgba 0.2066 0.6403 1.9 0.3>
      Morning Color=<rgba 0.2066 0.6403 1.9 0.3>
      Day Color=<rgba 0.2066 0.6403 1.9 0.3>
      Evening Color=<rgba 0.2066 0.6403 1.9 0.3>
     )
    )
    Lifetime Max=0.7
    Lifetime Min=0.55
    Lifetime Mode=::NewEnumerator1
    Mesh Scale=&Multiply_VectorByFloat(
     Vector=&MakeVector(X=0.8, Y=0.8, Z=&RandomRangeFloat(Minimum=1, Maximum=1.2032))
     Float=&RandomRangeFloat(Minimum=1.2621, Maximum=1.1112)
    )
    Mesh Scale Mode=::NewEnumerator4
    UsePositionOffset=false
    Write Material Random=true
   )
   &InitialMeshOrientation(
    Mesh Orientation Mode=::NewEnumerator2
    Orientation Coordinate Space=::Local
    Rotation=&RandomRangeVector(Minimum=<-0.318 -0.2963 0.1807>, Maximum=<-0.425 -0.2313 0.2096>)
    Use Rotation Vector=true
   )
   &SystemLocation(
    MaskSpawnGroup=false
    Offset=&RandomRangeVector(Minimum=<-2.8459 -30.2218 9.6377>, Maximum=<3.0034 -44.1743 21.702>)
   )
  ]
  particleUpdateScript [
   &ParticleState()
   &ScaleMeshSize(Scale Factor=&Multiply_VectorByFloat(
     Vector=&Multiply_VectorByFloat(
      Vector=&MakeVector(
       X=1
       Y=1
       Z=&FloatFromCurve(
        FloatCurve=DataInterface() {propertyValues {Curve {keys [{tangentMode=RCTM_User, time=0, value=1}, {tangentMode=RCTM_User, time=1, value=0.5}]}}}
        Scale Curve=&RandomRangeFloat(Minimum=0.8, Maximum=1)
       )
      )
      Float=0.8
     )
     Float=&FloatFromCurve(
      FloatCurve=DataInterface() {propertyValues {Curve {keys [
          {interpMode=RCIM_Cubic, tangentMode=RCTM_User, time=0, value=0.3}
          {arriveTangent=0.2897, interpMode=RCIM_Cubic, leaveTangent=0.2897, time=0.453, value=0.5392}
          {interpMode=RCIM_Cubic, tangentMode=RCTM_User, time=1, value=0.5897}
         ]}}}
      Scale Curve=3
     )
    ))
   &UpdateMeshOrientation(Rotation Vector=<0 0 1>)
   &DynamicMaterialParameters(
    Index 0 Param 1=&FloatFromCurve(FloatCurve=DataInterface() {propertyValues {Curve {keys [
         {interpMode=RCIM_Cubic, tangentMode=RCTM_User, time=0, value=0}
         {interpMode=RCIM_Cubic, tangentMode=RCTM_User, time=0.5, value=1}
         {arriveTangent=-4.73, tangentMode=RCTM_User, time=1, value=0}
        ]}}})
    Module.Param0WriteEnabled=true
    Module.Param1WriteEnabled=true
    Module.Param2WriteEnabled=true
    Module.Param3WriteEnabled=true
   ) {module=DynamicMaterialParameters001}
   &ScaleColor(
    Scale Alpha=&Multiply_Float(
     A=&FloatFromCurve(FloatCurve=DataInterface() {propertyValues {Curve {keys [
          {interpMode=RCIM_Cubic, tangentMode=RCTM_User, time=0, value=0}
          {interpMode=RCIM_Cubic, tangentMode=RCTM_User, time=0.2, value=1}
          {tangentMode=RCTM_User, time=1, value=0}
         ]}}})
     B=0.2
    )
    Scale RGB=&Multiply_VectorByFloat(
     Vector=&Multiply_VectorByFloat(Vector=<1 1 1>, Float=&FN_FrontEndFloatModifier(Gameplay=1, Front End=1.2))
     Float=1
    )
    ScaleA=true
    ScaleRGB=true
   )
  ]
  rendererClasses [&NiagaraMeshRendererProperties]
  renderers [{rendererClass=&NiagaraMeshRendererProperties, rendererIndex=0}]
 }
 NE_WindTrails {
  emitterUpdateScript [
   &EmitterState(Inactive Response=::NewEnumerator1, Life Cycle Mode=::NewEnumerator1, Loop Behavior=::NewEnumerator1, Loop Duration=3, UseLoopDelay=false)
   &FN_SpawnRibbon_v2(Number of Ribbons=5, Spawn Mode=::NewEnumerator0, SpawnRate=128)
  ]
  particleSpawnScript [
   &FN_SetSpawnGroup(Number of Spawn Groups=5)
   &FN_InitializeRibbon_v3(
    Color:LinearColor=&Multiply_LinearColorByFloat(
     Linear Color=&SelectLinearColorFromArray(
      Array Sampling Mode=::NewEnumerator1
      Direct Array Index=&SpawnGroup()
      Direct Array Mode=::NewEnumerator1
      Color Selection Array=DataInterface() {propertyValues {ColorData [<rgba 1 0.402 0.3743 1>, <rgba 0.2066 0.6403 1.9 0.3>]}}
      Random Seed=?
     )
     Float=&FNTimeOfDayScalar(Night Scale=0.7, Morning Scale=1, Evening Scale=0.8)
    )
    Custom Offset=true
    Lifetime:NiagaraFloat=&MaskFloatBySpawnGroup(
     Masked Float=&FloatFromCurve(
      FloatCurve=DataInterface() {propertyValues {Curve {keys [
          {interpMode=RCIM_Cubic, tangentMode=RCTM_User, time=0.5573, value=1}
          {arriveTangentWeight=0.2303, interpMode=RCIM_Cubic, tangentMode=RCTM_User, time=0.8938, value=0}
         ]}}}
      CurveIndex=$Emitter.Age
      Scale Curve=0.5
     )
     Passthrough Float=&MaskFloatBySpawnGroup(
      Masked Float=&FloatFromCurve(
       FloatCurve=DataInterface() {propertyValues {Curve {keys [
           {interpMode=RCIM_Cubic, tangentMode=RCTM_User, time=0.8087, value=1}
           {arriveTangentWeight=0.0936, interpMode=RCIM_Cubic, tangentMode=RCTM_User, time=1.1962, value=0}
          ]}}}
       CurveIndex=$Emitter.Age
       Scale Curve=0.5
      )
      Passthrough Float=&MaskFloatBySpawnGroup(
       Masked Float=&FloatFromCurve(
        FloatCurve=DataInterface() {propertyValues {Curve {keys [
            {interpMode=RCIM_Cubic, tangentMode=RCTM_User, time=0.5573, value=1}
            {arriveTangentWeight=0.2857, interpMode=RCIM_Cubic, tangentMode=RCTM_User, time=1.0026, value=0}
           ]}}}
        CurveIndex=$Emitter.Age
        Scale Curve=0.6
       )
       Passthrough Float=&MaskFloatBySpawnGroup(
        Masked Float=&FloatFromCurve(
         FloatCurve=DataInterface() {propertyValues {Curve {keys [
             {interpMode=RCIM_Cubic, tangentMode=RCTM_User, time=0.794, value=1}
             {arriveTangentWeight=0.1811, interpMode=RCIM_Cubic, tangentMode=RCTM_User, time=1.2466, value=0}
            ]}}}
         CurveIndex=$Emitter.Age
         Scale Curve=0.6
        )
        Passthrough Float=&FloatFromCurve(
         FloatCurve=DataInterface() {propertyValues {Curve {keys [
             {interpMode=RCIM_Cubic, tangentMode=RCTM_User, time=1.1211, value=1}
             {arriveTangentWeight=0.0927, interpMode=RCIM_Cubic, tangentMode=RCTM_User, time=1.45, value=0}
            ]}}}
         CurveIndex=$Emitter.Age
         Scale Curve=0.5
        )
        Spawn Group Mask=3
       )
       Spawn Group Mask=2
      )
      Spawn Group Mask=1
     )
    )
    Local Vector Offset:Vector3f=<0 1 0>
    Ribbon Width=20
   )
   &DynamicMaterialParameters(
    Index 0 Param 1=1
    Index 0 Param 3=1
    Index 0 Param 4=&Multiply_Float(
     A=&FloatFromCurve(
      FloatCurve=DataInterface() {propertyValues {Curve {keys [
          {interpMode=RCIM_Cubic, tangentMode=RCTM_User, time=0.1108, value=0}
          {arriveTangentWeight=0.0397, interpMode=RCIM_Cubic, tangentMode=RCTM_User, time=0.23, value=1}
         ]}}}
      CurveIndex=$Emitter.Age
     )
     B=&MaskFloatBySpawnGroup(
      Masked Float=&FloatFromCurve(
       FloatCurve=DataInterface() {propertyValues {Curve {keys [
           {interpMode=RCIM_Cubic, tangentMode=RCTM_User, time=0.7315, value=1}
           {arriveTangentWeight=0.2303, interpMode=RCIM_Cubic, tangentMode=RCTM_User, time=0.9581, value=0}
          ]}}}
       CurveIndex=$Emitter.Age
       Scale Curve=0.5
      )
      Passthrough Float=&MaskFloatBySpawnGroup(
       Masked Float=&FloatFromCurve(
        FloatCurve=DataInterface() {propertyValues {Curve {keys [
            {interpMode=RCIM_Cubic, tangentMode=RCTM_User, time=1.0596, value=1}
            {arriveTangentWeight=0.0936, interpMode=RCIM_Cubic, tangentMode=RCTM_User, time=1.1149, value=0}
           ]}}}
        CurveIndex=$Emitter.Age
        Scale Curve=0.5
       )
       Passthrough Float=&MaskFloatBySpawnGroup(
        Masked Float=&FloatFromCurve(
         FloatCurve=DataInterface() {propertyValues {Curve {keys [
             {interpMode=RCIM_Cubic, tangentMode=RCTM_User, time=0.5573, value=1}
             {arriveTangentWeight=0.2857, interpMode=RCIM_Cubic, tangentMode=RCTM_User, time=1.0026, value=0}
            ]}}}
         CurveIndex=$Emitter.Age
         Scale Curve=0.6
        )
        Passthrough Float=&MaskFloatBySpawnGroup(
         Masked Float=&FloatFromCurve(
          FloatCurve=DataInterface() {propertyValues {Curve {keys [
              {interpMode=RCIM_Cubic, tangentMode=RCTM_User, time=0.8452, value=1}
              {arriveTangent=-6.6259, arriveTangentWeight=0.1811, interpMode=RCIM_Cubic, leaveTangent=-6.6259, tangentMode=RCTM_User, time=1.2308, value=0}
             ]}}}
          CurveIndex=$Emitter.Age
          Scale Curve=0.6
         )
         Passthrough Float=&FloatFromCurve(
          FloatCurve=DataInterface() {propertyValues {Curve {keys [
              {interpMode=RCIM_Cubic, tangentMode=RCTM_User, time=1.2144, value=1}
              {arriveTangentWeight=0.0927, interpMode=RCIM_Cubic, tangentMode=RCTM_User, time=1.3988, value=0}
             ]}}}
          CurveIndex=$Emitter.Age
          Scale Curve=0.5
         )
         Spawn Group Mask=3
        )
        Spawn Group Mask=2
       )
       Spawn Group Mask=1
      )
     )
    )
    Module.Index0Param0Write=true
    Module.Index0Param1Write=true
    Module.Index0Param2Write=true
    Module.Index0Param3Write=true
   )
   &SkeletalMeshLocation(
    Apply Sampled Orientation to Mesh Particles=true
    Bone / Socket Index=$Particles.SpawnGroup
    Bone Sampling Mode=::NewEnumerator4
    Inherit Velocity=false
    Interpolate Mesh Sampling=false
    MaskSpawnGroup=false
    Sampled Position Offset=<0 0 3>
    Sampled Position Offset Coordinate Space:ENiagaraMeshTransforms=::NewEnumerator3
    Skeletal Mesh=$System.Ribbon Skeletal Mesh
   )
  ]
  particleUpdateScript [
   &ParticleState()
   &InheritSourceMovement(Applied Force Scale=<0.5 0.5 1>, Applied Position Scale=<0.5 0.5 1>, ApplyForce=false, ApplyPosition=true, ApplyPositionFalloff=false)
   &SolveForcesAndVelocity(Clamp Velocity=false, Limit Acceleration=false)
   &ScaleColor(
    Scale Alpha=&FloatFromCurve(
     FloatCurve=DataInterface() {propertyValues {Curve {keys [
         {arriveTangent=-0.1027, interpMode=RCIM_Cubic, leaveTangent=-0.1027, tangentMode=RCTM_User, time=0.1, value=1}
         {arriveTangent=-1.1111, interpMode=RCIM_Cubic, leaveTangent=-1.1111, time=0.3762, value=0.3007}
         {arriveTangentWeight=0.2079, interpMode=RCIM_Cubic, tangentMode=RCTM_User, time=1, value=0}
        ]}}}
     Scale Curve=0.5
    )
    ScaleA=true
    ScaleRGB=true
   ) {module=ScaleColor001}
   &ScaleRibbonWidth(Ribbon Width Scale=&FloatFromCurve(
     FloatCurve=DataInterface() {propertyValues {Curve {keys [
         {arriveTangent=-0.1027, interpMode=RCIM_Cubic, leaveTangent=-0.1027, tangentMode=RCTM_User, time=0.1, value=1}
         {arriveTangent=-1.1111, interpMode=RCIM_Cubic, leaveTangent=-1.1111, time=0.4313, value=0.7729}
         <at 1 0 RCIM_Cubic>
        ]}}}
     Scale Curve=0.7
    ))
   &FN_UpdateRibbon(Detach Locked Particle=true, Enable Skeletal Position=true) {module=FN_UpdateRibbon001}
  ]
  rendererClasses [&NiagaraRibbonRendererProperties]
  renderers [{rendererClass=&NiagaraRibbonRendererProperties}]
 }
 Smoke {
  emitterUpdateScript [
   &EmitterState(Life Cycle Mode=::NewEnumerator1, Loop Behavior=::NewEnumerator1, Loop Delay=0.2, Loop Duration=2, UseLoopDelay=true)
   &SpawnBurst_Instantaneous(Spawn Group=5, Spawn Time=1.212, Use Spawn Probability=false, UseLoopCountLimit=false) {module=SpawnBurst_Instantaneous005}
   &SpawnBurst_Instantaneous(Spawn Group=4, Spawn Time=1.1512, Use Spawn Probability=false, UseLoopCountLimit=false) {module=SpawnBurst_Instantaneous004}
   &SpawnBurst_Instantaneous(Spawn Group=3, Spawn Time=0.8532, Use Spawn Probability=false, UseLoopCountLimit=false) {module=SpawnBurst_Instantaneous003}
   &SpawnBurst_Instantaneous(Spawn Group=2, Spawn Time=1.0228, Use Spawn Probability=false, UseLoopCountLimit=false) {module=SpawnBurst_Instantaneous002}
   &SpawnBurst_Instantaneous(Spawn Group=1, Spawn Time=0.69, Use Spawn Probability=false, UseLoopCountLimit=false) {module=SpawnBurst_Instantaneous001}
   &SpawnBurst_Instantaneous(Spawn Count=1, Spawn Time=0.8556, Use Spawn Probability=false)
  ]
  particleSpawnScript [
   &InitializeParticle(
    Color=&FN_Select_Color_from_Time_Of_Day(Morning=<rgba 1 0.2915 0.308 0.49>, Day=<rgba 1 0.2731 0.2808 0.5>, Evening=<rgba 1 0.2988 0.3508 0.42>, Night=<rgba 1 0.2915 0.308 0.49>)
    Lifetime=&UniformRangedFloat(Minimum=0.1, Maximum=0.2)
    Sprite Rotation Angle=&UniformRangedFloat(Minimum=0, Maximum=1)
    Sprite Rotation Mode=::NewEnumerator1
    Sprite Size Mode=::NewEnumerator3
    Uniform Sprite Size=&UniformRangedFloat(Minimum=15, Maximum=10)
    UsePositionOffset=false
    Write Material Random=true
   ) {module=InitializeParticle001}
   /SetVariables(Particles.InitSpriteRotRate=&UniformRangedFloat(Minimum=-1, Maximum=1)) {isSetParameters=true}
   &SkeletalMeshLocation(
    Bone / Socket Index=&SpawnGroup()
    Bone Sampling Mode=::NewEnumerator4
    MaskSpawnGroup=false
    Skeletal Mesh=$System.Skeletal Mesh
   )
   &ApplyInitialForces()
   &CameraOffset(Camera Offset Amount=&RandomRangeFloat(Minimum=15, Maximum=20))
  ]
  particleUpdateScript [
   &ParticleState()
   &SpriteRotationRate(Rotation Rate=&Multiply_Float(
     A=$Particles.InitSpriteRotRate
     B=&FloatFromCurve(
      FloatCurve=DataInterface() {propertyValues {Curve {keys [
          {arriveTangent=-1.9214, interpMode=RCIM_Cubic, leaveTangent=-1.9214, tangentMode=RCTM_User, time=-0.0046, value=1.0113}
          {arriveTangent=-0.2496, interpMode=RCIM_Cubic, leaveTangent=-0.2496, tangentMode=RCTM_User, time=0.9896, value=0.16}
         ]}}}
      CurveIndex=$Emitter.Age
     )
    ))
   &ScaleColor(
    Scale Alpha=&FloatFromCurve(FloatCurve=DataInterface() {propertyValues {Curve {keys [<at 0 0>,<at 0.2049 0.9765>,<at 0.7791 0.9107>,<at 1 0>]}}})
    ScaleA=true
    ScaleRGB=true
   ) {module=ScaleColor001}
   &SolveForcesAndVelocity(Clamp Velocity=false, Limit Acceleration=false)
   &ScaleSpriteSize(Scale Factor=&Vector2DFromFloat(Value=&FloatFromCurve(FloatCurve=DataInterface() {propertyValues {Curve {keys [
          {arriveTangent=5.2354, interpMode=RCIM_Cubic, leaveTangent=5.2354, tangentMode=RCTM_User, time=-0.0046, value=0.25}
          <at 0.1439 1.4923>
          <at 1.0011 3.4809 RCIM_Cubic>
         ]}}})))
   &DynamicMaterialParameters(
    Index 0 Param 1=&FloatFromCurve(FloatCurve=DataInterface() {propertyValues {Curve {keys [<at 0 0 RCIM_Cubic>,<at 1.0011 0.2>]}}})
    Index 0 Param 2=&FloatFromCurve(FloatCurve=DataInterface() {propertyValues {Curve {keys [
         {arriveTangent=0.0453, interpMode=RCIM_Cubic, tangentMode=RCTM_Break, time=0, value=-1}
         {arriveTangent=0.0604, interpMode=RCIM_Cubic, leaveTangent=0.0604, tangentMode=RCTM_User, time=1, value=-1}
        ]}}})
    Index 0 Param 3=&FloatFromCurve(FloatCurve=DataInterface() {propertyValues {Curve {keys [{interpMode=RCIM_Cubic, leaveTangent=4.6779, time=0.2851, value=0}]}}})
    Index 1 Param 1=&FloatFromCurve(
     FloatCurve=DataInterface() {propertyValues {Curve {keys [<at 0 0>,<at 0.25 1>]}}}
     Scale Curve=&RandomRangeFloat(Minimum=2, Maximum=1)
    )
    Index 1 Param 2=&FloatFromCurve(
     FloatCurve=DataInterface() {propertyValues {Curve {keys [
         {interpMode=RCIM_Cubic, leaveTangent=3.7422, tangentMode=RCTM_Break, time=0, value=0.2}
         {interpMode=RCIM_Cubic, leaveTangent=0.9066, time=1, value=2}
        ]}}}
     Scale Curve=&RandomRangeFloat(Minimum=0.5, Maximum=1)
    )
    Index 1 Param 3=&FloatFromCurve(FloatCurve=DataInterface() {propertyValues {Curve {keys [
         {interpMode=RCIM_Cubic, leaveTangent=2.8445, tangentMode=RCTM_Break, time=0, value=0.2}
         {interpMode=RCIM_Cubic, leaveTangent=63.2978, time=0.25, value=1}
        ]}}})
    Index 1 Param 4=&FloatFromCurve(
     FloatCurve=DataInterface() {propertyValues {Curve {keys [<at 0 1>,<at 1 0.8>]}}}
     Scale Curve=&RandomRangeFloat(Minimum=0.5, Maximum=1)
    )
    Index 2 Param 1=&FloatFromCurve(FloatCurve=DataInterface() {propertyValues {Curve {keys [{interpMode=RCIM_Cubic, leaveTangent=3.2309, tangentMode=RCTM_Break, time=0.1, value=0}, <at 1 1>]}}})
    Index 2 Param 2=$Particles.MaterialRandom
    Index 2 Param 3=&FloatFromCurve(FloatCurve=DataInterface() {propertyValues {Curve {keys [
         {interpMode=RCIM_Cubic, tangentMode=RCTM_Break, time=0, value=1}
         {interpMode=RCIM_Cubic, leaveTangent=11.1233, tangentMode=RCTM_Break, time=1, value=2}
        ]}}})
    Index 2 Param 4=0.75
    Module.Param0WriteEnabled=true
    Module.Param0WriteEnabled1=true
    Module.Param0WriteEnabled2=true
    Module.Param1WriteEnabled=false
    Module.Param1WriteEnabled1=true
    Module.Param1WriteEnabled2=true
    Module.Param2WriteEnabled=true
    Module.Param2WriteEnabled1=true
    Module.Param2WriteEnabled2=true
    Module.Param3WriteEnabled=true
    Module.Param3WriteEnabled1=true
    Module.Param3WriteEnabled2=true
    Write Parameter Index 1=true
    Write Parameter Index 2=true
   )
  ]
  rendererClasses [&NiagaraSpriteRendererProperties]
  renderers [{rendererClass=&NiagaraSpriteRendererProperties}]
 }
}
summary {
 systemName=NS_Emote_CopiousCranes_Wind
 emitters [
  {emitterName=NE_WindDisc, rendererClasses [&NiagaraMeshRendererProperties]}
  {emitterName=NE_WindStreaks, rendererClasses [&NiagaraMeshRendererProperties]}
  {emitterName=NE_Sparkles, rendererClasses [&NiagaraSpriteRendererProperties]}
  {emitterName=Smoke, rendererClasses [&NiagaraSpriteRendererProperties]}
  {emitterName=NE_WindTrails, rendererClasses [&NiagaraRibbonRendererProperties]}
 ]
 userVariables [
  User.bIsFrontEnd {type=NiagaraBool}
  User.bIsFrontEndPreview {type=NiagaraBool}
  User.RibbonIDOffset {type=NiagaraInt32}
  User.SkeletalMesh {
   type=NiagaraDataInterfaceSkeletalMesh
   defaultValue {
    struct=NiagaraExt_VariableValue_DataInterface
    value {dataInterface=&self:NiagaraDataInterfaceSkeletalMesh_0}
   }
  }
 ]
}
